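# Physics asset for the MED_CornerWeek backpack cosmetic, defining collision and joint limits for two dangling weight chains.
# The pack bone carries a compound body of large and small capsules as the collision volume at the chain roots.
# Chain A has three capsule links and chain B four, each ending in a longer, thicker capsule.
# Constraints link each chain back to the pack with tight twist and swing that loosens toward the tips.
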PhysicsAsset /BRCosmetics/Accessories/FORT_Backpacks/Backpack_M_MED_CornerWeek/Meshes/M_MED_CornerWeek_Pack_Physics_FE
bodies {
 pack [
  {shapeType=Capsule, length=18.3496, radius=28.677, center=<0 22.3894 -30.6241>, rotation=<rot -0.0005 -90 90>}
  {shapeType=Capsule, length=8, radius=4.5, shapeName=pack_capsule, center=<-9.5767 -5.5292 -27.3445>, rotation=<rot -0.1317 0.7478 -80.0141>}
  {shapeType=Capsule, length=6, radius=4.5, shapeName=pack_capsule, center=<9.4402 -5.595 -29.0864>, rotation=<rot -0.5033 2.1166 -76.9643>}
  {shapeType=Capsule, length=18.5003, radius=11.4997, shapeName=pack_capsule, center=<-16.1876 -4.7592 -11.0153>}
  {shapeType=Capsule, length=18.0003, radius=11.9997, shapeName=pack_capsule, center=<16.5718 -3.6324 -11.4044>}
  {shapeType=Capsule, length=21.6788, radius=28.1112, shapeName=pack_capsule, center=<0 -35.2712 -5.1518>, rotation=<rot 90 0.0001 0>}
  {shapeType=Capsule, length=19.5606, radius=28.2298, shapeName=pack_capsule, center=<0 21.8162 -16.7845>}
 ]
 weight_a [{shapeType=Capsule, length=2.6672, radius=0.505, shapeName=weight_a_capsule, center=<1.947 -0.0017 0.0017>, rotation=<rot 0 89.9491 -89.9493>}]
 weight_a_01 [{
   shapeType=Capsule
   length=2.2193
   radius=0.505
   shapeName=weight_a_01_capsule
   center=<1.6861 0.0072 -0.0015>
   rotation=<rot 0 89.9772 -89.9017>
  }]
 weight_a_02 [{
   shapeType=Capsule
   length=9.5574
   radius=1.211
   shapeName=weight_a_02_capsule
   center=<5.5317 -0.0084 0.0081>
   rotation=<rot 0 89.9108 -89.9132>
  }]
 weight_b [{shapeType=Capsule, length=1.969, radius=0.505, shapeName=weight_b_capsule, center=<1.5003 0.0007 0.0005>, rotation=<rot 0 90.0274 -89.9823>}]
 weight_b_01 [{
   shapeType=Capsule
   length=2.6983
   radius=0.505
   shapeName=weight_b_01_capsule
   center=<2.0036 0.0113 -0.0038>
   rotation=<rot 0 -89.8198 -89.1879>
  }]
 weight_b_02 [{
   shapeType=Capsule
   length=2.485
   radius=0.505
   shapeName=weight_b_02_capsule
   center=<1.8762 0.0072 0.0077>
   rotation=<rot 0 89.9791 -89.9096>
  }]
 weight_b_03 [{
   shapeType=Capsule
   length=8.0441
   radius=1.211
   shapeName=weight_b_03_capsule
   center=<4.9011 -0.0085 0.0054>
   rotation=<rot 0 89.8986 -89.9352>
  }]
}
constraints [
 {bone1Name=weight_a, bone2Name=pack, swing1LimitDegrees=7, swing2LimitDegrees=13, twistLimitDegrees=2}
 {bone1Name=weight_b, bone2Name=pack, swing1LimitDegrees=7, swing2LimitDegrees=10, twistLimitDegrees=2}
 {bone1Name=weight_a_01, bone2Name=weight_a, swing1LimitDegrees=18, swing2LimitDegrees=13, twistLimitDegrees=2}
 {bone1Name=weight_a_02, bone2Name=weight_a_01, swing1LimitDegrees=28, swing2LimitDegrees=13, twistLimitDegrees=2}
 {bone1Name=weight_b_01, bone2Name=weight_b, swing1LimitDegrees=18, swing2LimitDegrees=13, twistLimitDegrees=2}
 {bone1Name=weight_b_02, bone2Name=weight_b_01, swing1LimitDegrees=18, swing2LimitDegrees=13, twistLimitDegrees=2}
 {bone1Name=weight_b_03, bone2Name=weight_b_02, swing1LimitDegrees=28, swing2LimitDegrees=13, twistLimitDegrees=2}
]
