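# Animation Blueprint for the CornerWeek backpack cosmetic, animating its clock hands and pendulum and simulating physics on its dangling parts.
# A cached main pose offsets the pack from spine bend, springs several root bones with AnimDynamics, and applies rotator variables to clock and pendulum bones.
# That pose feeds two RigidBody and AnimDynamics branches, selected by front-end state, with different physics assets and angular limits.
# A deprecated state function updates RigidBody gravity each frame.
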AnimBlueprint /BRCosmetics/Accessories/FORT_Backpacks/Backpack_M_MED_CornerWeek/Meshes/M_MED_CornerWeek_Pack_AnimBP
target_skeleton=&/M_MED_CornerWeek_Pack_Skeleton
graphs [
 AnimGraph {nodes [
   n1 = RigidBody!(
    ComponentPose=Local To Component(LocalPose=Use cached pose 'Main pose'())
    Alpha
    OverrideWorldGravity:Vector=Get GravityOverride()
    ExternalForce
   ) -> (Pose) {
    bDefaultToSkeletalMeshPhysicsAsset=false
    bOverrideWorldGravity=true
    overridePhysicsAsset=&/M_MED_CornerWeek_Pack_Physics_FE
    alphaBoolBlend {blendOption=Linear}
    componentAppliedLinearAccClamp=<500 500 500>
    componentLinearAccScale=<1 1 1>
    componentLinearVelScale=<8 8 8>
    overrideWorldGravity {z=-980}
   }
   n2 = Anim Dynamics - Bone: roots_a_1!(SimulationSpace:AnimPhysSimSpaceType Enum, ComponentPose=@n1, Alpha=1, GravityOverride:Vector="0,0,-980", ExternalForce="0,0,0") -> (Pose) {
    angularDampingOverride=0.999
    bAngularSpring=true
    bOverrideAngularDamping=true
    bUseGravityOverride=true
    alphaBoolBlend {blendOption=Linear}
    boundBone {boneName=roots_a_1}
    componentAppliedLinearAccClamp=<3000 3000 2500>
    componentLinearAccScale=<0.5 0.5 0.5>
    componentLinearVelScale=<0.5 0.5 0.5>
    physicsBodyDefinitions [{
      boundBone {boneName=roots_a_1}
      boxExtents=<20 20 20>
      constraintSetup {coneAngle=1, angularLimitsMax=<1 21 0>, angularLimitsMin=<-1 0 -26>, angularTarget=<0 0 0>}
      localJointOffset=<1 0 0>
     }]
   }
   n3 = Anim Dynamics - Bone: roots_a_2!(SimulationSpace:AnimPhysSimSpaceType Enum, ComponentPose=@n2, Alpha=1, GravityOverride:Vector="0,0,-980", ExternalForce="0,0,0") -> (Pose) {
    angularDampingOverride=0.97
    angularSpringConstant=100
    bAngularSpring=true
    bOverrideAngularDamping=true
    bUseGravityOverride=true
    alphaBoolBlend {blendOption=Linear}
    boundBone {boneName=roots_a_2}
    componentAppliedLinearAccClamp=<3000 3000 2500>
    componentLinearAccScale=<0.5 0.5 0.5>
    componentLinearVelScale=<0.5 0.5 0.5>
    physicsBodyDefinitions [{
      boundBone {boneName=roots_a_2}
      boxExtents=<15 15 15>
      constraintSetup {coneAngle=1, angularLimitsMax=<3 17 17>, angularLimitsMin=<-3 0 0>, angularTarget=<0 0 0>}
      localJointOffset=<1 0 0>
     }]
   }
   n4 = Anim Dynamics - Bone: roots_a_3!(SimulationSpace:AnimPhysSimSpaceType Enum, ComponentPose=@n3, Alpha=1, GravityOverride:Vector="0,0,-980", ExternalForce="0,0,0") -> (Pose) {
    angularDampingOverride=0.97
    angularSpringConstant=300
    bAngularSpring=true
    bOverrideAngularDamping=true
    bUseGravityOverride=true
    alphaBoolBlend {blendOption=Linear}
    boundBone {boneName=roots_a_3}
    componentAppliedLinearAccClamp=<3000 3000 2500>
    componentLinearAccScale=<0.5 0.5 0.5>
    componentLinearVelScale=<0.5 0.5 0.5>
    physicsBodyDefinitions [{
      boundBone {boneName=roots_a_3}
      boxExtents=<10 10 10>
      constraintSetup {coneAngle=1, angularLimitsMax=<3 0 8>, angularLimitsMin=<-3 -21 -8>, angularTarget=<0 0 0>}
      localJointOffset=<1 0 0>
     }]
   }
   n5 = Anim Dynamics - Bone: roots_b_1!(SimulationSpace:AnimPhysSimSpaceType Enum, ComponentPose=@n4, Alpha=8, GravityOverride:Vector="0,0,-980", ExternalForce="0,0,0") -> (Pose) {
    angularDampingOverride=0.999
    bAngularSpring=true
    bOverrideAngularDamping=true
    bUseGravityOverride=true
    alphaBoolBlend {blendOption=Linear}
    boundBone {boneName=roots_b_1}
    componentAppliedLinearAccClamp=<3000 3000 2500>
    componentLinearAccScale=<0.5 0.5 0.5>
    componentLinearVelScale=<0.5 0.5 0.5>
    physicsBodyDefinitions [{
      boundBone {boneName=roots_b_1}
      boxExtents=<25 25 25>
      constraintSetup {coneAngle=1, angularLimitsMax=<1 13 13>, angularLimitsMin=<-1 0 0>, angularTarget=<0 0 0>}
      localJointOffset=<1 0 0>
     }]
   }
   n6 = Anim Dynamics - Bone: roots_b_2!(SimulationSpace:AnimPhysSimSpaceType Enum, ComponentPose=@n5, Alpha=0.8, GravityOverride:Vector="0,0,-980", ExternalForce="0,0,0") -> (Pose) {
    angularDampingOverride=0.99
    angularSpringConstant=50
    bAngularSpring=true
    bOverrideAngularDamping=true
    bUseGravityOverride=true
    alphaBoolBlend {blendOption=Linear}
    boundBone {boneName=roots_b_2}
    componentAppliedLinearAccClamp=<3000 3000 2500>
    componentLinearAccScale=<0.5 0.5 0.5>
    componentLinearVelScale=<0.5 0.5 0.5>
    physicsBodyDefinitions [{
      boundBone {boneName=roots_b_2}
      boxExtents=<16 16 16>
      constraintSetup {coneAngle=1, angularLimitsMax=<3 13 13>, angularLimitsMin=<-3 0 0>, angularTarget=<0 0 0>}
      localJointOffset=<1 0 0>
     }]
   }
   n7 = Anim Dynamics - Bone: roots_b_3!(SimulationSpace:AnimPhysSimSpaceType Enum, ComponentPose=@n6, Alpha=1, GravityOverride:Vector="0,0,-980", ExternalForce="0,0,0") -> (Pose) {
    angularDampingOverride=0.99
    angularSpringConstant=150
    bAngularSpring=true
    bOverrideAngularDamping=true
    bUseGravityOverride=true
    alphaBoolBlend {blendOption=Linear}
    boundBone {boneName=roots_b_3}
    componentAppliedLinearAccClamp=<3000 3000 2500>
    componentLinearAccScale=<0.5 0.5 0.5>
    componentLinearVelScale=<0.5 0.5 0.5>
    physicsBodyDefinitions [{
      boundBone {boneName=roots_b_3}
      boxExtents=<10 10 10>
      constraintSetup {coneAngle=1, angularLimitsMax=<3 13 13>, angularLimitsMin=<-3 0 0>, angularTarget=<0 0 0>}
      localJointOffset=<1 0 0>
     }]
   }
   n8 = RigidBody(ComponentPose=Local To Component(LocalPose=Use cached pose 'Main pose'()), OverrideWorldGravity=Get GravityOverride()) {
    bDefaultToSkeletalMeshPhysicsAsset=false
    bOverrideWorldGravity=true
    overridePhysicsAsset=&/M_MED_CornerWeek_Pack_Physics
    alphaBoolBlend {blendOption=Linear}
    componentAppliedLinearAccClamp=<500 500 500>
    componentLinearAccScale=<1 1 1>
    componentLinearVelScale=<8 8 8>
    overrideWorldGravity {z=-980}
   }
   n9 = Anim Dynamics - Bone: roots_a_1(ComponentPose=@n8, Alpha=1, GravityOverride="0,0,-980", ExternalForce="0,0,0") {
    angularDampingOverride=0.999
    bAngularSpring=true
    bOverrideAngularDamping=true
    bUseGravityOverride=true
    alphaBoolBlend {blendOption=Linear}
    boundBone {boneName=roots_a_1}
    componentAppliedLinearAccClamp=<3000 3000 2500>
    componentLinearAccScale=<0.5 0.5 0.5>
    componentLinearVelScale=<0.5 0.5 0.5>
    physicsBodyDefinitions [{
      boundBone {boneName=roots_a_1}
      boxExtents=<20 20 20>
      constraintSetup {coneAngle=1, angularLimitsMax=<1 30 0>, angularLimitsMin=<-1 0 -35>, angularTarget=<0 0 0>}
      localJointOffset=<1 0 0>
     }]
   }
   n10 = Anim Dynamics - Bone: roots_a_2(ComponentPose=@n9, Alpha=1, GravityOverride="0,0,-980", ExternalForce="0,0,0") {
    angularDampingOverride=0.97
    angularSpringConstant=100
    bAngularSpring=true
    bOverrideAngularDamping=true
    bUseGravityOverride=true
    alphaBoolBlend {blendOption=Linear}
    boundBone {boneName=roots_a_2}
    componentAppliedLinearAccClamp=<3000 3000 2500>
    componentLinearAccScale=<0.5 0.5 0.5>
    componentLinearVelScale=<0.5 0.5 0.5>
    physicsBodyDefinitions [{
      boundBone {boneName=roots_a_2}
      boxExtents=<15 15 15>
      constraintSetup {coneAngle=1, angularLimitsMax=<3 25 25>, angularLimitsMin=<-3 0 0>, angularTarget=<0 0 0>}
      localJointOffset=<1 0 0>
     }]
   }
   n11 = Anim Dynamics - Bone: roots_a_3(ComponentPose=@n10, Alpha=1, GravityOverride="0,0,-980", ExternalForce="0,0,0") {
    angularDampingOverride=0.97
    angularSpringConstant=300
    bAngularSpring=true
    bOverrideAngularDamping=true
    bUseGravityOverride=true
    alphaBoolBlend {blendOption=Linear}
    boundBone {boneName=roots_a_3}
    componentAppliedLinearAccClamp=<3000 3000 2500>
    componentLinearAccScale=<0.5 0.5 0.5>
    componentLinearVelScale=<0.5 0.5 0.5>
    physicsBodyDefinitions [{
      boundBone {boneName=roots_a_3}
      boxExtents=<10 10 10>
      constraintSetup {coneAngle=1, angularLimitsMax=<3 0 10>, angularLimitsMin=<-3 -30 0>, angularTarget=<0 0 0>}
      localJointOffset=<1 0 0>
     }]
   }
   n12 = Anim Dynamics - Bone: roots_b_1(ComponentPose=@n11, Alpha=8, GravityOverride="0,0,-980", ExternalForce="0,0,0") {
    angularDampingOverride=0.999
    bAngularSpring=true
    bOverrideAngularDamping=true
    bUseGravityOverride=true
    alphaBoolBlend {blendOption=Linear}
    boundBone {boneName=roots_b_1}
    componentAppliedLinearAccClamp=<3000 3000 2500>
    componentLinearAccScale=<0.5 0.5 0.5>
    componentLinearVelScale=<0.5 0.5 0.5>
    physicsBodyDefinitions [{
      boundBone {boneName=roots_b_1}
      boxExtents=<25 25 25>
      constraintSetup {coneAngle=1, angularLimitsMax=<1 20 20>, angularLimitsMin=<-1 0 0>, angularTarget=<0 0 0>}
      localJointOffset=<1 0 0>
     }]
   }
   n13 = Anim Dynamics - Bone: roots_b_2(ComponentPose=@n12, Alpha=0.8, GravityOverride="0,0,-980", ExternalForce="0,0,0") {
    angularDampingOverride=0.99
    angularSpringConstant=50
    bAngularSpring=true
    bOverrideAngularDamping=true
    bUseGravityOverride=true
    alphaBoolBlend {blendOption=Linear}
    boundBone {boneName=roots_b_2}
    componentAppliedLinearAccClamp=<3000 3000 2500>
    componentLinearAccScale=<0.5 0.5 0.5>
    componentLinearVelScale=<0.5 0.5 0.5>
    physicsBodyDefinitions [{
      boundBone {boneName=roots_b_2}
      boxExtents=<16 16 16>
      constraintSetup {coneAngle=1, angularLimitsMax=<3 20 20>, angularLimitsMin=<-3 0 0>, angularTarget=<0 0 0>}
      localJointOffset=<1 0 0>
     }]
   }
   n14 = Anim Dynamics - Bone: roots_b_3(ComponentPose=@n13, Alpha=1, GravityOverride="0,0,-980", ExternalForce="0,0,0") {
    angularDampingOverride=0.99
    angularSpringConstant=150
    bAngularSpring=true
    bOverrideAngularDamping=true
    bUseGravityOverride=true
    alphaBoolBlend {blendOption=Linear}
    boundBone {boneName=roots_b_3}
    componentAppliedLinearAccClamp=<3000 3000 2500>
    componentLinearAccScale=<0.5 0.5 0.5>
    componentLinearVelScale=<0.5 0.5 0.5>
    physicsBodyDefinitions [{
      boundBone {boneName=roots_b_3}
      boxExtents=<10 10 10>
      constraintSetup {coneAngle=1, angularLimitsMax=<3 20 20>, angularLimitsMin=<-3 0 0>, angularTarget=<0 0 0>}
      localJointOffset=<1 0 0>
     }]
   }
   Output Pose(Result=Blend Poses by bool(BlendPose_0=Component To Local(ComponentPose=@n7), BlendPose_1=Component To Local(ComponentPose=@n14), bActiveValue=Get bIsInFrontEnd())) {layerGroup=None}
   n15 = "Transform (Modify) Bone - Bone: pack"!(ComponentPose=Local To Component(LocalPose=Input Pose()), bAlphaBoolEnabled:bool, Translation="0,0,0", Rotation="0,0,8", Scale) -> (Pose) {
    alphaInputType=Bool
    rotationMode=BMM_Additive
    translationMode=BMM_Additive
    alphaBoolBlend {blendOption=Linear}
    boneToModify {boneName=pack}
    rotation {roll=8}
   }
   n16 = "Transform (Modify) Bone - Bone: pack"!(ComponentPose=@n15, Alpha:float, Translation=Get backpack_translation(), Rotation=Get backpack_rotate(), Scale) -> (Pose) {rotationMode=BMM_Additive, translationMode=BMM_Additive, alphaBoolBlend {blendOption=Linear}, boneToModify {boneName=pack}}
   n17 = "Transform (Modify) Bone - Bone: pack"(ComponentPose=@n16, Translation="0,0,0", Rotation=Get Backpack_rotate_X() -> (Backpack_rotate_X:Rotator)) {rotationMode=BMM_Additive, translationMode=BMM_Additive, alphaBoolBlend {blendOption=Linear}, boneToModify {boneName=pack}}
   n18 = Anim Dynamics - Bone: roots_c!(SimulationSpace:AnimPhysSimSpaceType Enum, ComponentPose=@n17, Alpha=0.8, GravityOverride:Vector="0,0,-980", ExternalForce="0,0,0") -> (Pose) {
    angularDampingOverride=0.9999
    bAngularSpring=true
    bOverrideAngularDamping=true
    bUseGravityOverride=true
    alphaBoolBlend {blendOption=Linear}
    boundBone {boneName=roots_c}
    componentAppliedLinearAccClamp=<3000 3000 2500>
    componentLinearAccScale=<0.5 0.5 0.5>
    componentLinearVelScale=<0.5 0.5 0.5>
    physicsBodyDefinitions [{
      boundBone {boneName=roots_c}
      boxExtents=<13 13 13>
      constraintSetup {coneAngle=1, angularLimitsMax=<3 0 0>, angularLimitsMin=<-3 -35 -35>, angularTarget=<0 0 0>}
      localJointOffset=<1 0 0>
     }]
   }
   n19 = Anim Dynamics - Bone: roots_d!(SimulationSpace:AnimPhysSimSpaceType Enum, ComponentPose=@n18, Alpha=0.8, GravityOverride:Vector="0,0,-980", ExternalForce="0,0,0") -> (Pose) {
    angularDampingOverride=0.9999
    angularSpringConstant=600
    bAngularSpring=true
    bOverrideAngularDamping=true
    bUseGravityOverride=true
    alphaBoolBlend {blendOption=Linear}
    boundBone {boneName=roots_d}
    componentAppliedLinearAccClamp=<3000 3000 2500>
    componentLinearAccScale=<0.5 0.5 0.5>
    componentLinearVelScale=<0.5 0.5 0.5>
    physicsBodyDefinitions [{
      boundBone {boneName=roots_d}
      boxExtents=<20 20 20>
      constraintSetup {coneAngle=1, angularLimitsMax=<3 0 0>, angularLimitsMin=<-3 -25 -25>, angularTarget=<0 0 0>}
      localJointOffset=<1 0 0>
     }]
   }
   n20 = Anim Dynamics - Bone: roots_e!(SimulationSpace:AnimPhysSimSpaceType Enum, ComponentPose=@n19, Alpha=1, GravityOverride:Vector="0,0,-980", ExternalForce="0,0,0") -> (Pose) {
    angularDampingOverride=0.999
    bAngularSpring=true
    bOverrideAngularDamping=true
    bUseGravityOverride=true
    alphaBoolBlend {blendOption=Linear}
    boundBone {boneName=roots_e}
    componentAppliedLinearAccClamp=<3000 3000 2500>
    componentLinearAccScale=<0.5 0.5 0.5>
    componentLinearVelScale=<0.5 0.5 0.5>
    physicsBodyDefinitions [{
      boundBone {boneName=roots_e}
      boxExtents=<15 15 15>
      constraintSetup {coneAngle=1, angularLimitsMax=<3 0 30>, angularLimitsMin=<-3 -36 0>, angularTarget=<0 0 0>}
      localJointOffset=<1 0 0>
     }]
   }
   n21 = Anim Dynamics - Bone: roots_f!(SimulationSpace:AnimPhysSimSpaceType Enum, ComponentPose=@n20, Alpha=1, GravityOverride:Vector="0,0,-980", ExternalForce="0,0,0") -> (Pose) {
    angularDampingOverride=0.99
    bAngularSpring=true
    bOverrideAngularDamping=true
    bUseGravityOverride=true
    alphaBoolBlend {blendOption=Linear}
    boundBone {boneName=roots_f}
    componentAppliedLinearAccClamp=<3000 3000 2500>
    componentLinearAccScale=<0.5 0.5 0.5>
    componentLinearVelScale=<0.5 0.5 0.5>
    physicsBodyDefinitions [{
      boundBone {boneName=roots_f}
      boxExtents=<10 10 10>
      constraintSetup {coneAngle=1, angularLimitsMax=<0 35 0>, angularLimitsMin=<0 0 0>, angularTarget=<0 0 0>}
      localJointOffset=<1 0 0>
     }]
   }
   n22 = Anim Dynamics - Bone: roots_g!(SimulationSpace:AnimPhysSimSpaceType Enum, ComponentPose=@n21, Alpha=1, GravityOverride:Vector="0,0,-980", ExternalForce="0,0,0") -> (Pose) {
    angularDampingOverride=0.999
    bAngularSpring=true
    bOverrideAngularDamping=true
    bUseGravityOverride=true
    alphaBoolBlend {blendOption=Linear}
    boundBone {boneName=roots_g}
    componentAppliedLinearAccClamp=<3000 3000 2500>
    componentLinearAccScale=<0.5 0.5 0.5>
    componentLinearVelScale=<0.5 0.5 0.5>
    physicsBodyDefinitions [{
      boundBone {boneName=roots_g}
      boxExtents=<10 10 10>
      constraintSetup {coneAngle=1, angularLimitsMax=<3 0 30>, angularLimitsMin=<-3 -30 0>, angularTarget=<0 0 0>}
      localJointOffset=<1 0 0>
     }]
   }
   n23 = "Transform (Modify) Bone - Bone: pendulum"(ComponentPose=@n22, Alpha=1, Translation="0,0,0", Rotation=Get PendulumRotator() -> (PendulumRotator:Rotator), Scale="1,1,1") {rotationMode=BMM_Replace, rotationSpace=BCS_BoneSpace, alphaBoolBlend {blendOption=Linear}, boneToModify {boneName=pendulum}}
   n24 = "Transform (Modify) Bone - Bone: clock_a"(ComponentPose=@n23, Alpha=1, Translation="0,0,0", Rotation="49.413,-0.000,-0.001", Scale="1,1,1") {
    rotationMode=BMM_Additive
    rotationSpace=BCS_BoneSpace
    alphaBoolBlend {blendOption=Linear}
    boneToModify {boneName=clock_a}
    rotation=<rot 49.413 -0.0003 -0.0014>
   }
   n25 = "Transform (Modify) Bone - Bone: clock_a"(ComponentPose=@n24, Alpha=1, Translation="0,0,0", Rotation=Get HourHandRotator() -> (HourHandRotator:Rotator), Scale="1,1,1") {rotationMode=BMM_Additive, rotationSpace=BCS_BoneSpace, alphaBoolBlend {blendOption=Linear}, boneToModify {boneName=clock_a}}
   n26 = "Transform (Modify) Bone - Bone: clock_b"(ComponentPose=@n25, Alpha=1, Translation="0,0,0", Rotation="-68.135,0,0", Scale="1,1,1") {
    rotationMode=BMM_Replace
    rotationSpace=BCS_BoneSpace
    alphaBoolBlend {blendOption=Linear}
    boneToModify {boneName=clock_b}
    rotation {pitch=-68.135}
   }
   n27 = "Transform (Modify) Bone - Bone: clock_b"(ComponentPose=@n26, Alpha=1, Translation="0,0,0", Rotation=Get MinuteHandRotator() -> (MinuteHandRotator:Rotator), Scale="1,1,1") {rotationMode=BMM_Replace, rotationSpace=BCS_BoneSpace, alphaBoolBlend {blendOption=Linear}, boneToModify {boneName=clock_b}}
   Save cached pose 'Main pose'(Pose=Component To Local(ComponentPose=@n27)) {cacheName=Main pose}
  ]}
 EventGraph {nodes [
   Event BlueprintUpdateAnimation()
   > n1 = Sequence()
   > n2 = Gravity Override State Function DEPRECATED(
    CharacterPartAnimInstance=Self-Reference()
    GameTimeInSeconds=GetGameTimeInSeconds()
    GravityOverrideParams=Get Gravity Override Params()
    RBAN Params=Get RBAN Params()
   )
   > Set GravityOverride(GravityOverride=@n2)
   n3 = GetOwnerSkeletalMeshComponent()
   n4 = BreakRotator(InRot=BreakTransform(InTransform=GetDeltaTransformFromRefPose(self=@n3, BoneName=spine_05, BaseName=spine_01)).Rotation)
   n5 = MapRangeClamped(Value=@n4.Yaw, InRangeA=0, InRangeB=-150, OutRangeB=175)
   Set backpack_rotate(execute=@n1.then_1, backpack_rotate=MakeRotator(Roll=@n5, Pitch=float / float(A=@n5, B=5)))
   > Set backpack_translation(backpack_translation=MakeVector(Y=MapRangeClamped(Value=@n4.Yaw, InRangeA=0, InRangeB=-50, OutRangeA=-0.5, OutRangeB=-10)))
   > Set Backpack_rotate_X(Backpack_rotate_X:Rotator=MakeRotator(Roll=MapRangeClamped(Value=@n4, InRangeA=-20, InRangeB=-45, OutRangeB=13))) -> (Output_Get:Rotator)
   > n6 = Branch(Condition=Get bIsInFrontEnd())
   > Set SimlationSpace(SimlationSpace:AnimPhysSimSpaceType Enum=Component) -> (Output_Get:AnimPhysSimSpaceType Enum)
   Set SimlationSpace(execute=@n6.else)
   n7 = Event OnCharacterCustomizationLoaded()
   > n8 = Cast To FortPawn!(execute, Object=GetOwningActor()) -> (then, CastFailed, AsFort Pawn:Fort Pawn Object Reference) {is=title K2Node_DynamicCast}
   n9 = Get bIsPlayingEmote()
   n10 = Get GenderPermitted!(self:Custom Character Part Object Reference=Get Array Item with EFort Custom Part Type(Array:Array of Custom Character Part Object References=@n7.CustomParts, PartType=Head) -> (Output:Custom Character Part Object Reference)) -> (GenderPermitted:EFortCustomGender Enum=Invalid)
   > Set Is Idle Female(Is Idle Female:bool=AND Boolean!(A="Equal (Enum)"(A:EFortCustomGender Enum=@n10, B:EFortCustomGender Enum=Female), B=Get bIsInFrontEnd(), C:bool=NOT Boolean(A=@n9)) -> (ReturnValue)) -> (Output_Get:bool)
   TryGetPawnOwner()
   n11 = GetCurrentMontage!(self:Character Object Reference=@n8) -> (ReturnValue:Anim Montage Object Reference)
   GetObjectName(Object=@n11)
   OR Boolean!(
    A="Equal (Object)"(A=@n11, B=/BRCosmetics/Animation/Game/MainPlayer/Menu/BR/Female_LobbyIdle_HandsDownWide_Montage)
    B="Equal (Object)"(A=@n11, B=/BRCosmetics/Animation/Game/MainPlayer/Menu/BR/CustomSkins/_s37/SaxoPop/SaxoPop_FE_idle_Montage)
    C:bool
   ) -> (ReturnValue)
  ]}
]
variables [
 backpack_rotate {type=Rotator, default=<rot 1 0 5>}
 Backpack_rotate_X {type=Rotator}
 backpack_translation {type=Vector, default=<0 -0.9091 0>}
 Foat_pick_axe {type=number}
 Gravity Override Params {
  type=GravityOverrideParamsStruct
  default {
   skydive_GoingUp {gravityOverride=<0 -300 -980>}
   skydive_Parachuting {gravityOverride=<0 -300 -980>, windAmplitude=<100 -200 200>, windFrequency=<1.5 3 8>}
  }
 }
 GravityOverride {type=Vector}
 Is Idle Female {type=boolean}
 NewVar {type="array<Rotator>", default []}
 NewVar_0 {type=string, default=""}
 Parashuting {type=number}
 RBAN Params {
  type=RBANParamsStruct
  default {
   dBNO {simSpaceSettings {
     maxAngularAcceleration=1000
     maxAngularVelocity=3
     maxLinearAcceleration=300
     maxLinearVelocity=700
     velocityScaleZ=0
     externalLinearDragV=<0 0.5 0>
    }}
   dBNO_Carried {simSpaceSettings {
     maxAngularAcceleration=1000
     maxAngularVelocity=3
     maxLinearAcceleration=300
     maxLinearVelocity=700
     velocityScaleZ=0
     externalLinearDragV=<0 0.5 0>
    }}
   emoteOrMelee {simSpaceSettings {
     maxAngularAcceleration=1000
     maxAngularVelocity=3
     maxLinearAcceleration=300
     maxLinearVelocity=700
     velocityScaleZ=0
     externalLinearDragV=<0 0.5 0>
    }}
   front_End {simSpaceSettings {
     maxAngularAcceleration=1000
     maxAngularVelocity=3
     maxLinearAcceleration=300
     maxLinearVelocity=700
     velocityScaleZ=0
     externalLinearDragV=<0 0.5 0>
    }}
   onGround_Crouch_Moving {simSpaceSettings {
     maxAngularAcceleration=1000
     maxAngularVelocity=3
     maxLinearAcceleration=300
     maxLinearVelocity=700
     velocityScaleZ=0
     externalLinearDragV=<0 0.5 0>
    }}
   onGround_Falling {simSpaceSettings {
     maxAngularAcceleration=1000
     maxAngularVelocity=3
     maxLinearAcceleration=300
     maxLinearVelocity=700
     velocityScaleZ=0
     externalLinearDragV=<0 0.5 0>
    }}
   onGround_Moving {simSpaceSettings {
     maxAngularAcceleration=1000
     maxAngularVelocity=3
     maxLinearAcceleration=300
     maxLinearVelocity=700
     velocityScaleZ=0
     externalLinearDragV=<0 0.5 0>
    }}
   onGround_Standing {simSpaceSettings {
     maxAngularAcceleration=1000
     maxAngularVelocity=3
     maxLinearAcceleration=300
     maxLinearVelocity=700
     velocityScaleZ=0
     externalLinearDragV=<0 0.5 0>
    }}
   rail_Grinding {simSpaceSettings {
     maxAngularAcceleration=1000
     maxAngularVelocity=3
     maxLinearAcceleration=300
     maxLinearVelocity=700
     velocityScaleZ=0
     worldAlpha=0.1
     externalLinearDragV=<0 0.5 0>
    }}
   riding_Moving {simSpaceSettings {
     maxAngularAcceleration=1000
     maxAngularVelocity=3
     maxLinearAcceleration=300
     maxLinearVelocity=700
     velocityScaleZ=0
     externalLinearDragV=<0 0.5 0>
    }}
   riding_Moving_Targeting {simSpaceSettings {
     maxAngularAcceleration=1000
     maxAngularVelocity=3
     maxLinearAcceleration=300
     maxLinearVelocity=700
     velocityScaleZ=0
     externalLinearDragV=<0 0.5 0>
    }}
   riding_Standing {simSpaceSettings {
     maxAngularAcceleration=1000
     maxAngularVelocity=3
     maxLinearAcceleration=300
     maxLinearVelocity=700
     velocityScaleZ=0
     externalLinearDragV=<0 0.5 0>
    }}
   skydive_DivingDown {simSpaceSettings {
     maxAngularAcceleration=1000
     maxAngularVelocity=3
     maxLinearAcceleration=300
     maxLinearVelocity=700
     velocityScaleZ=0
     externalLinearDragV=<0 0.5 0>
    }}
   skydive_FloatingDown {
    componentLinearAccScale=<0 0.5 0>
    componentLinearVelScale=<0 0.5 0>
    simSpaceSettings {
     maxAngularAcceleration=1000
     maxAngularVelocity=3
     maxLinearAcceleration=300
     maxLinearVelocity=700
     velocityScaleZ=0
     externalLinearDragV=<0 0.5 0>
    }
   }
   skydive_GoingUp {simSpaceSettings {
     maxAngularAcceleration=1000
     maxAngularVelocity=3
     maxLinearAcceleration=300
     maxLinearVelocity=700
     velocityScaleZ=0
     externalLinearDragV=<0 0.5 0>
    }}
   skydive_Parachuting {
    componentLinearAccScale=<0 0.1 0.3>
    componentLinearVelScale=<0 0.1 0.3>
    simSpaceSettings {
     maxAngularAcceleration=1000
     maxAngularVelocity=3
     maxLinearAcceleration=300
     maxLinearVelocity=700
     velocityScaleZ=0
     externalLinearDragV=<0 0.5 0>
    }
   }
   skydive_Surfing {simSpaceSettings {
     maxAngularAcceleration=1000
     maxAngularVelocity=3
     maxLinearAcceleration=300
     maxLinearVelocity=700
     velocityScaleZ=0
     externalLinearDragV=<0 0.5 0>
    }}
   swimming {simSpaceSettings {
     maxAngularAcceleration=1000
     maxAngularVelocity=3
     maxLinearAcceleration=300
     maxLinearVelocity=700
     velocityScaleZ=0
     externalLinearDragV=<0 0.5 0>
    }}
   tactical_Sprint {simSpaceSettings {
     maxAngularAcceleration=1000
     maxAngularVelocity=3
     maxLinearAcceleration=300
     maxLinearVelocity=700
     velocityScaleZ=0
     externalLinearDragV=<0 0.5 0>
    }}
   ziplining {simSpaceSettings {
     maxAngularAcceleration=1000
     maxAngularVelocity=3
     maxLinearAcceleration=300
     maxLinearVelocity=700
     velocityScaleZ=0
     externalLinearDragV=<0 0.5 0>
    }}
  }
 }
 RotationMin {type=Rotator}
 SimlationSpace {type=AnimPhysSimSpaceType}
]
