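# Physics asset for the pendulum variant of the CornerWeek backpack cosmetic, giving its dangling pendulum piece a physical setup.
# It has two capsule bodies, one on the pack bone and one offset along the pendulum_01 bone.
# A single constraint joins pendulum_01 to pack, with twist locked and swing limits set for a narrow arc on one axis and zero on the other.
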PhysicsAsset /BRCosmetics/Accessories/FORT_Backpacks/Backpack_M_MED_CornerWeek/Meshes/M_MED_CornerWeek_Pack_Phy_Pendulum
bodies {
 pack [{shapeType=Capsule, length=12.2923, radius=7.8188, center=<-0.7311 -5.7398 -5.2941>, rotation=<rot -2.623 -90 -89.8396>}]
 pendulum_01 [{shapeType=Capsule, length=14.5746, radius=3.393, shapeName=pendulum_01_capsule, center=<14.5987 0 0>, rotation=<rot 0 90 -89.9999>}]
}
constraints [{bone1Name=pendulum_01, bone2Name=pack, swing1LimitDegrees=0, swing2LimitDegrees=10, twistLimitDegrees=45, twistMotion=Locked}]
